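# Physics asset for the keychain on the Donut Plate backpack cosmetic, defining collision bodies and joints for its dangling chain.
# Four box-bodied links (parent, 01, 02, 03) are constrained in series from dyn_keychain_base, with wide swing and very tight twist limits, slightly narrower on the last link.
# Large capsules on the base bone and a thin one on pack add collision volumes; pack and the small keyring-base box have no constraints.
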PhysicsAsset /BRCosmetics/Accessories/FORT_Backpacks/Backpack_F_MED_Donut_Plate/Meshes/F_MED_Donut_Plate_KeyChain_Physics
bodies {
 dyn_cloth_keyring_base [{shapeType=Box, extentX=0.5, extentY=0.5, extentZ=0.5}]
 dyn_keychain_01 [{shapeType=Box, extentX=2.51, extentY=4.7681, extentZ=2.2712, center=<1.2326 0 0>}]
 dyn_keychain_02 [{shapeType=Box, extentX=2.51, extentY=4.7681, extentZ=2.2712, center=<1.2326 0 0>}]
 dyn_keychain_03 [{shapeType=Box, extentX=9.6893, extentY=9.42, extentZ=2.125, center=<4.978 0 0>}]
 dyn_keychain_base [
  {shapeType=Capsule, length=44.3213, radius=8.6036, center=<6.137 0.0005 14.863>, rotation=<rot 0 0 -90>}
  {shapeType=Capsule, length=46.8676, radius=10.9323, center=<15.7075 0.0007 15.4599>, rotation=<rot 0 0 -90>}
  {shapeType=Capsule, length=44.8876, radius=12.9123, center=<3.9096 0.0007 17.9663>, rotation=<rot -5.2291 0 -89.9999>}
  {shapeType=Box, extentX=0.75, extentY=0.75, extentZ=0.75}
 ]
 dyn_keychain_parent [{shapeType=Box, extentX=2.51, extentY=4.7681, extentZ=2.2712, center=<1.2326 0 0>}]
 pack [{shapeType=Capsule, length=24.6743, radius=1.8381, center=<-19.9652 -2.5973 -25.9289>, rotation=<rot -55.8215 0 -2.7962>}]
}
constraints [
 {bone1Name=dyn_keychain_parent, bone2Name=dyn_keychain_base, swing1LimitDegrees=55, swing2LimitDegrees=45, twistLimitDegrees=3}
 {bone1Name=dyn_keychain_01, bone2Name=dyn_keychain_parent, swing1LimitDegrees=55, swing2LimitDegrees=45, twistLimitDegrees=3}
 {bone1Name=dyn_keychain_02, bone2Name=dyn_keychain_01, swing1LimitDegrees=55, swing2LimitDegrees=45, twistLimitDegrees=3}
 {bone1Name=dyn_keychain_03, bone2Name=dyn_keychain_02, swing1LimitDegrees=55, swing2LimitDegrees=35, twistLimitDegrees=3}
]
