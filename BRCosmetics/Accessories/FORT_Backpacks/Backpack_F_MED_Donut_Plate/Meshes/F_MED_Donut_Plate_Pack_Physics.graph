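# Physics asset for the Donut Plate back bling cosmetic, defining bodies and joints for its dangling cloth and keyring bones.
# A cloth chain of small-to-larger boxes runs from dyn_cloth_base through dyn_cloth_parent to dyn_cloth_06, linked by angle-limited constraints.
# A separate, tighter-limited constraint joins the keyring parent to its base.
# Capsules on the keychain base and pack bones have no constraints.
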PhysicsAsset /BRCosmetics/Accessories/FORT_Backpacks/Backpack_F_MED_Donut_Plate/Meshes/F_MED_Donut_Plate_Pack_Physics
bodies {
 dyn_cloth_01 [{shapeType=Box, extentX=1.9736, extentY=2.3856, extentZ=2.9049, center=<0.9835 0 0>}]
 dyn_cloth_02 [{shapeType=Box, extentX=3.7236, extentY=2.3856, extentZ=4.6549, center=<1.9048 0 0>}]
 dyn_cloth_03 [{shapeType=Box, extentX=5.7236, extentY=2.3856, extentZ=7.6549, center=<2.6806 0 0>}]
 dyn_cloth_04 [{shapeType=Box, extentX=5.7236, extentY=2.3856, extentZ=7.6549, center=<2.6806 0 0>}]
 dyn_cloth_05 [{shapeType=Box, extentX=5.7236, extentY=2.3856, extentZ=7.6549, center=<2.6806 0 0>}]
 dyn_cloth_06 [{shapeType=Box, extentX=5.7236, extentY=2.3856, extentZ=7.6549, center=<2.6806 0 0>}]
 dyn_cloth_base [{shapeType=Box, extentX=0.75, extentY=0.75, extentZ=0.75}]
 dyn_cloth_keyring_base [{shapeType=Box, extentX=0.5, extentY=0.5, extentZ=0.5}]
 dyn_cloth_keyring_parent [{shapeType=Box, extentX=3.7686, extentY=3.3112, extentZ=1.6785, center=<1.7728 0 0>}]
 dyn_cloth_parent [{shapeType=Box, extentX=1.7236, extentY=2.0106, extentZ=3.5299, center=<0.8384 0 0>}]
 dyn_keychain_base [
  {shapeType=Capsule, length=44.3213, radius=8.6036, center=<14.8719 0.0005 17.1361>, rotation=<rot 0 0 -90>}
  {shapeType=Capsule, length=46.8676, radius=10.9323, center=<24.4424 0.0007 17.733>, rotation=<rot 0 0 -90>}
  {shapeType=Capsule, length=44.8876, radius=12.9123, center=<12.6445 0.0007 20.2393>, rotation=<rot -5.2291 0 -89.9999>}
 ]
 pack [{shapeType=Capsule, length=23.6118, radius=2.9006, center=<-19.9652 -2.5973 -25.9289>, rotation=<rot -55.8215 0 -2.7962>}]
}
constraints [
 {bone1Name=dyn_cloth_keyring_parent, bone2Name=dyn_cloth_keyring_base, swing1LimitDegrees=10, swing2LimitDegrees=30, twistLimitDegrees=3}
 {bone1Name=dyn_cloth_parent, bone2Name=dyn_cloth_base, swing1LimitDegrees=30, swing2LimitDegrees=30, twistLimitDegrees=10}
 {bone1Name=dyn_cloth_01, bone2Name=dyn_cloth_parent, swing1LimitDegrees=30, swing2LimitDegrees=30, twistLimitDegrees=10}
 {bone1Name=dyn_cloth_02, bone2Name=dyn_cloth_01, swing1LimitDegrees=30, swing2LimitDegrees=30, twistLimitDegrees=10}
 {bone1Name=dyn_cloth_03, bone2Name=dyn_cloth_02, swing1LimitDegrees=30, swing2LimitDegrees=30, twistLimitDegrees=10}
 {bone1Name=dyn_cloth_04, bone2Name=dyn_cloth_03, swing1LimitDegrees=30, swing2LimitDegrees=30, twistLimitDegrees=10}
 {bone1Name=dyn_cloth_05, bone2Name=dyn_cloth_04, swing1LimitDegrees=30, swing2LimitDegrees=30, twistLimitDegrees=10}
 {bone1Name=dyn_cloth_06, bone2Name=dyn_cloth_05, swing1LimitDegrees=30, swing2LimitDegrees=30, twistLimitDegrees=10}
]
