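# Physics asset for the sword straps on the Donut Plate backpack cosmetic (F_MED), simulating two dynamic strap chains.
# Each strap segment is a box body, with small boxes at the left and right base bones.
# Constraints chain each segment to the one above it, from base through parent to the numbered bones.
# Swing and twist limits are tight, and the longer right chain swings less per joint than the left.
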PhysicsAsset /BRCosmetics/Accessories/FORT_Backpacks/Backpack_F_MED_Donut_Plate/Meshes/F_MED_Donut_Plate_Pack_Sword_Physics
bodies {
 dyn_sword_strap_l_01 [{shapeType=Box, extentX=5.4435, extentY=5.4638, extentZ=3.75, center=<2.6733 0 0>}]
 dyn_sword_strap_l_base [{shapeType=Box, extentX=1, extentY=1, extentZ=1}]
 dyn_sword_strap_l_parent [{shapeType=Box, extentX=5.4173, extentY=5.383, extentZ=2.25, center=<2.6464 0 0>}]
 dyn_sword_strap_r_01 [{shapeType=Box, extentX=5.4173, extentY=5.383, extentZ=2.25, center=<2.6464 0 0>}]
 dyn_sword_strap_r_03 [{shapeType=Box, extentX=5.4173, extentY=5.383, extentZ=2.25, center=<2.6464 0 0>}]
 dyn_sword_strap_r_base [{shapeType=Box, extentX=0.5, extentY=0.5, extentZ=0.5}]
 dyn_sword_strap_r_parent [{shapeType=Box, extentX=5.4173, extentY=5.383, extentZ=2.25, center=<2.6464 0 0>}]
}
constraints [
 {bone1Name=dyn_sword_strap_l_parent, bone2Name=dyn_sword_strap_l_base, swing1LimitDegrees=10, swing2LimitDegrees=10, twistLimitDegrees=3}
 {bone1Name=dyn_sword_strap_l_01, bone2Name=dyn_sword_strap_l_parent, swing1LimitDegrees=10, swing2LimitDegrees=10, twistLimitDegrees=3}
 {bone1Name=dyn_sword_strap_r_parent, bone2Name=dyn_sword_strap_r_base, swing1LimitDegrees=5, swing2LimitDegrees=5, twistLimitDegrees=3}
 {bone1Name=dyn_sword_strap_r_01, bone2Name=dyn_sword_strap_r_parent, swing1LimitDegrees=5, swing2LimitDegrees=5, twistLimitDegrees=3}
 {bone1Name=dyn_sword_strap_r_03, bone2Name=dyn_sword_strap_r_01, swing1LimitDegrees=5, swing2LimitDegrees=5, twistLimitDegrees=3}
]
