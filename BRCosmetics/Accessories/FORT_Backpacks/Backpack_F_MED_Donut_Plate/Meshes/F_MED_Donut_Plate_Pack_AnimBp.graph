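# Animation Blueprint for the Donut Plate backpack cosmetic, adding secondary motion to its sword, cloth and keychain bones.
# Spine yaw drives an additive root offset feeding three chained RigidBody simulations, with constraints anchoring keyring and keychain bases to sword straps.
# Jumping or falling additively tilts the keychain.
# When rigid bodies are disabled or LOD exceeds 1, output switches to cheaper AnimDynamics on cloth and keychain, without that root offset.
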AnimBlueprint /BRCosmetics/Accessories/FORT_Backpacks/Backpack_F_MED_Donut_Plate/Meshes/F_MED_Donut_Plate_Pack_AnimBp
target_skeleton=/Game/Accessories/FORT_Backpacks/Backpack_F_MED_Donut_Plate/Meshes/F_MED_Donut_Plate_Pack_Skeleton
graphs [
 AnimGraph {nodes [
   n1 = "Transform (Modify) Bone - Bone: root"(
    ComponentPose=Local To Component(LocalPose=Use cached pose 'DynamicsCache'())
    Translation=Get backpack_translation()
    Rotation=Get backpack_rotate()
   ) {
    rotationMode=BMM_Additive
    rotationSpace=BCS_BoneSpace
    translationMode=BMM_Additive
    translationSpace=BCS_BoneSpace
    alphaBoolBlend {blendOption=Linear}
    boneToModify {boneName=root}
   }
   n2 = RigidBody(ComponentPose=@n1) {
    bClampLinearTranslationLimitToRefPose=true
    bOverrideWorldGravity=true
    overridePhysicsAsset=&/F_MED_Donut_Plate_Pack_Sword_Physics
    simulationSpace=BaseBoneSpace
    alphaBoolBlend {blendOption=Linear}
    alphaScaleBiasClamp {bInterpResult=true, interpSpeedDecreasing=6, interpSpeedIncreasing=7}
    baseBoneRef {boneName=root}
    componentAppliedLinearAccClamp=<3000 3000 3000>
    componentLinearAccScale=<0.25 0.25 0.25>
    componentLinearVelScale=<0.25 0.25 0.25>
    overrideWorldGravity {z=-980}
   }
   n3 = "Constraint - dyn_cloth_keyring_base (PP)"!(ComponentPose=@n2, Alpha=1, ConstraintWeights_0=0.5, ConstraintWeights_1:float=0.5) -> (Pose) {
    alphaBoolBlend {blendOption=Linear}
    boneToModify {boneName=dyn_cloth_keyring_base}
    constraintSetup [{targetBone {boneName=dyn_sword_strap_l_01}}, {targetBone {boneName=dyn_sword_strap_r_03}}]
    constraintWeights [1,1]
   }
   n4 = "Constraint - dyn_keychain_base (PP)"!(ComponentPose=@n3, Alpha=1, ConstraintWeights_0=0.5, ConstraintWeights_1:float=0.5) -> (Pose) {
    alphaBoolBlend {blendOption=Linear}
    boneToModify {boneName=dyn_keychain_base}
    constraintSetup [{targetBone {boneName=dyn_sword_strap_l_01}}, {targetBone {boneName=dyn_sword_strap_r_03}}]
    constraintWeights [1,1]
   }
   n5 = RigidBody(ComponentPose=@n4) {
    bClampLinearTranslationLimitToRefPose=true
    bOverrideWorldGravity=true
    overridePhysicsAsset=&/F_MED_Donut_Plate_Pack_Physics
    simulationSpace=BaseBoneSpace
    alphaBoolBlend {blendOption=Linear}
    alphaScaleBiasClamp {bInterpResult=true, interpSpeedDecreasing=6, interpSpeedIncreasing=7}
    baseBoneRef {boneName=root}
    componentAppliedLinearAccClamp=<3000 3000 3000>
    componentLinearAccScale=<0.15 0.15 0.15>
    componentLinearVelScale=<0.15 0.15 0.15>
    overrideWorldGravity {z=-500}
   }
   n6 = RigidBody(ComponentPose=@n5) {
    bClampLinearTranslationLimitToRefPose=true
    bOverrideWorldGravity=true
    overridePhysicsAsset=&/F_MED_Donut_Plate_KeyChain_Physics
    simulationSpace=BaseBoneSpace
    alphaBoolBlend {blendOption=Linear}
    alphaScaleBiasClamp {bInterpResult=true, bias=0.1, interpSpeedDecreasing=6, interpSpeedIncreasing=5, scale=0.8}
    baseBoneRef {boneName=root}
    componentAppliedLinearAccClamp=<3000 3000 3000>
    componentLinearAccScale=<0.15 0.15 0.15>
    componentLinearVelScale=<0.15 0.15 0.15>
    overrideWorldGravity {z=-980}
   }
   n7 = "Transform (Modify) Bone - Bone: dyn_keychain_parent"(ComponentPose=@n6, Alpha=Get Jump_Alpha() -> (Jump_Alpha:double), Translation="0,0,0", Rotation="-40,0,0", Scale="1,1,1") {
    rotationMode=BMM_Additive
    rotationSpace=BCS_BoneSpace
    translationMode=BMM_Additive
    translationSpace=BCS_BoneSpace
    alphaBoolBlend {blendOption=Linear}
    alphaScaleBiasClamp {bInterpResult=true, interpSpeedDecreasing=2, interpSpeedIncreasing=5}
    boneToModify {boneName=dyn_keychain_parent}
    rotation {pitch=-40}
   }
   n8 = "Anim Dynamics - Chain: dyn_cloth_01 -> dyn_cloth_06"(ComponentPose=Local To Component(LocalPose=Use cached pose 'DynamicsCache'()), Alpha=1, ExternalForce="0,0,0") {
    angularDampingOverride=0.9999
    bChain=true
    bGravityOverrideInSimSpace=true
    bOverrideLinearDamping=true
    gravityScale=0.1
    linearDampingOverride=0.9999
    numSolverIterationsPostUpdate=10
    numSolverIterationsPreUpdate=10
    alphaBoolBlend {blendOption=Linear}
    boundBone {boneName=dyn_cloth_01}
    chainEnd {boneName=dyn_cloth_06}
    componentLinearAccScale=<0.25 0.25 0.25>
    componentLinearVelScale=<0.25 0.25 0.25>
    gravityOverride=<0.999 0.999 0.999>
    physicsBodyDefinitions [{
      sphereCollisionRadius=0
      boundBone {boneName=dyn_cloth_01}
      boxExtents=<6 6 4>
      constraintSetup {angularLimitsMax=<0 10 20>, angularLimitsMin=<0 -10 -20>, angularTarget=<0 0 0>}
      localJointOffset=<4 0 0>
     }]
    relativeSpaceBone {boneName=root}
   }
   n9 = Anim Dynamics - Bone: dyn_keychain_01(ComponentPose=@n8, Alpha=1, ExternalForce="0,0,0") {
    angularDampingOverride=0.9999
    bGravityOverrideInSimSpace=true
    bOverrideLinearDamping=true
    gravityScale=0.1
    linearDampingOverride=0.9999
    numSolverIterationsPostUpdate=10
    numSolverIterationsPreUpdate=10
    alphaBoolBlend {blendOption=Linear}
    boundBone {boneName=dyn_keychain_01}
    chainEnd {boneName=dyn_keychain_03}
    componentAppliedLinearAccClamp=<6000 6000 6000>
    componentLinearAccScale=<0.25 0.25 0.25>
    componentLinearVelScale=<0.25 0.25 0.25>
    gravityOverride=<0.999 0.999 0.999>
    physicsBodyDefinitions [{
      sphereCollisionRadius=0
      boundBone {boneName=dyn_keychain_01}
      boxExtents=<6 6 4>
      constraintSetup {angularLimitsMax=<0 10 20>, angularLimitsMin=<0 -20 -20>, angularTarget=<0 0 0>}
      localJointOffset=<4 0 0>
     }]
    relativeSpaceBone {boneName=root}
   }
   Output Pose(Result=Blend Poses by bool(
     BlendPose_0=Component To Local(ComponentPose=@n9)
     BlendPose_1=Component To Local(ComponentPose=@n7)
     bActiveValue=Get isRigidBodyDisabled() -> (isRigidBodyDisabled:bool)
    ) {blendType=Linear}) {layerGroup=None}
   Save cached pose 'DynamicsCache'(Pose=Input Pose()) {cacheName=DynamicsCache}
   Trail controller - Bone: dyn_cloth_02!(
    ComponentPose
    bAlphaBoolEnabled:bool=Get isRigidBodyDisabled()
    RotationOffsets_0:Vector="0, 0, 0"
    RotationOffsets_1:Vector="0, 0, 0"
    RotationOffsets_2:Vector="0, 0, 0"
    RotationOffsets_3:Vector="0, 0, 0"
   ) -> (Pose) {
    alphaInputType=Bool
    bActorSpaceFakeVel=true
    bEnableDebug=true
    bLimitRotation=true
    bReorientParentToChild=false
    bShowTrailLocation=true
    bUsePlanarLimit=true
    chainLength=4
    lastBoneRotationAnimAlphaBlend=0.25
    maxDeltaTime=0.0333
    relaxationSpeedScale=0.25
    stretchLimit=0.3
    alphaScaleBiasClamp {bInterpResult=true, interpSpeedDecreasing=6, interpSpeedIncreasing=8}
    planarLimits [{drivingBone {boneName=dyn_cloth_parent}, planeTransform {location=<0 0 0>, rotation=<rot 0 -90 90>, scale=<1 1 1>}}]
    relaxationSpeedScaleInputProcessor {bInterpResult=true, interpSpeedDecreasing=5, interpSpeedIncreasing=6}
    rotationLimits [
     {limitMax=<60 60 60>, limitMin=<-60 -60 -60>}
     {limitMax=<60 60 60>, limitMin=<-60 -60 -60>}
     {limitMax=<60 60 60>, limitMin=<-60 -60 -60>}
     {limitMax=<60 60 60>, limitMin=<-60 -60 -60>}
    ]
    rotationOffsets [<0 0 0>, <100 100 100>, <100 100 100>, <100 100 100>]
    trailBone {boneName=dyn_cloth_02}
    trailRelaxationSpeed {editorCurveData {keys [<at 0 10 RCIM_Linear>,<at 0.5894 8.8535 RCIM_Linear>,<at 0.9268 7.7255 RCIM_Linear>]}}
   }
  ]}
 EventGraph {nodes [
   Event BlueprintUpdateAnimation()
   > n1 = Sequence!(execute) -> (then_0, then_1, then_2:Exec, then_3:Exec)
   n2 = GetOwnerSkeletalMeshComponent()
   n3 = BreakRotator(InRot=BreakTransform(InTransform=GetDeltaTransformFromRefPose(self=@n2, BoneName=spine_05, BaseName=spine_01)).Rotation)
   > Set backpack_rotate(backpack_rotate=MakeRotator(Roll=MapRangeClamped(Value=@n3.Yaw, InRangeA=0, InRangeB=-30, OutRangeB=20)))
   > Set backpack_translation(backpack_translation=MakeVector(Y=MapRangeClamped(Value=@n3.Yaw, InRangeA=0, InRangeB=-30, OutRangeB=-3)))
   n4 = Get PawnSpeed2D()
   Set movement_alpha(execute=@n1.then_1, movement_alpha:double=SelectFloat(A=0, B=0, bPickA="float >= float"(A=@n4, B=0) {is=title K2Node_CallFunction})) -> (Output_Get:double)
   n5 = Get bIsJumping()
   n6 = Get bIsFalling()
   Set Jump_Alpha(execute=@n1.then_3, Jump_Alpha:double=SelectFloat(A=1, B=0, bPickA=OR Boolean(A=@n5, B=@n6))) -> (Output_Get:double)
   n7 = Get CurrentLOD()
   n8 = Get bRigidBodyEnabled()
   Set isRigidBodyDisabled(
    execute=@n1.then_2
    isRigidBodyDisabled:bool=OR Boolean(A=NOT Boolean(A=@n8), B="integer > integer"(A=@n7, B=1) {is=title K2Node_CallFunction})
   ) -> (Output_Get:bool)
   TryGetPawnOwner()
  ]}
]
variables [
 backpack_rotate {type=Rotator}
 backpack_translation {type=Vector}
 isRigidBodyDisabled {type=boolean}
 Jump_Alpha {type=number}
 movement_alpha {type=number}
]
